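; Looping gameplay cue for the Train Heist that plays braking effects along the whole train.
; On looping start, it casts the cue target to BaseCar and gets that car's train manager.
; If the manager is valid, it spawns the brake Niagara system attached to the root component of every car in the train.
; The graph does not store or stop the spawned systems, and its other events are empty.
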(event EventBeginPlay)

(event Collision|EventActorBeginOverlap
  (params (OtherActor :type Actor%20Object%20Reference)))

(event EventOnLoopingStart(EitherVFX-preferthisversion)
  (params (MyTarget :type Actor%20Object%20Reference) (Parameters :type Gameplay%20Cue%20Parameters%20Structure%20%28by%20ref%29) (ParticleComponents :type Array%20of%20FXSystem%20Component%20Object%20References) (AudioComponents :type Array%20of%20Audio%20Component%20Object%20References))
  (bind _asbase_car (Utilities|Casting|CastToBaseCar MyTarget))
  (Utilities|IsValid (Class|TrainCar|GetTrainManager _asbase_car)
    (:Is%20Valid
      (for _array_element (Variables|/Script/TrainCoreRuntime.FortTrainManager|GetCars (Class|TrainCar|GetTrainManager _asbase_car))
        (Niagara|SpawnSystemAttached "/TrainHeist/FX/NS_Train_Heist_BrakeFX.NS_Train_Heist_BrakeFX" (Variables|/Script/Engine.Actor|GetRootComponent (Utilities|Struct|BreakTrainCarRuntimeData _array_element)))))))

(event EventTick
  (params (DeltaSeconds :type Float%20%28single-precision%29)))

(fn ConstructionScript)
