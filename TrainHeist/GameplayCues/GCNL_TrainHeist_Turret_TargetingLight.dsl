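; Looping gameplay cue for the Train Heist turret's targeting light, with audio that follows the turret.
; When looping starts, it runs the parent logic, then attaches each valid audio component to the cue's target attach component at a small local offset.
; It casts the target to the on-train sentry turret and, if the tracked player pawn is locally viewed, spawns an extra tracked-warning sound on that attach component.
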(event EventBeginPlay)

(event Collision|EventActorBeginOverlap
  (params (OtherActor :type Actor%20Object%20Reference)))

(event EventTick
  (params (DeltaSeconds :type Float%20%28single-precision%29)))

(event EventOnLoopingStart(EitherVFX-preferthisversion)
  (params (MyTarget :type Actor%20Object%20Reference) (Parameters :type Gameplay%20Cue%20Parameters%20Structure%20%28by%20ref%29) (ParticleComponents :type Array%20of%20FXSystem%20Component%20Object%20References) (AudioComponents :type Array%20of%20Audio%20Component%20Object%20References))
  (bind (_normalizedmagnitude _rawmagnitude _effectcontext _matchedtagname _originaltag _aggregatedsourcetags _aggregatedtargettags _location _normal _instigator _effectcauser _sourceobject _physicalmaterial _gameplayeffectlevel _abilitylevel _targetattachcomponent) (Ability|GameplayCue|BreakGameplayCueParameters Parameters))
  (|CallParentFunction MyTarget Parameters ParticleComponents AudioComponents)
  (for _array_element AudioComponents
    (Utilities|IsValid _array_element
      (:Is%20Valid
        (Transformation|AttachComponentToComponent _array_element _targetattachcomponent "None" "KeepRelative" "KeepRelative" "KeepRelative" true)
        (Transformation|SetRelativeLocation _array_element "12.000000,0.000000,20.000000" false false))))
  (bind _asbga_athena_sentry_turret_on_train (Utilities|Casting|CastToBGA_Athena_Sentry_Turret_OnTrain MyTarget))
  (bind _asplayer_pawn_athena (Utilities|Casting|CastToPlayerPawn_Athena (Variables|/Script/FortniteGame.BuildingGameplayActorSentry|GetActorBeingTracked _asbga_athena_sentry_turret_on_train)))
  (if (Pawn|IsLocallyViewed _asplayer_pawn_athena)
    (Audio|SpawnSoundAttached (Variables|Default|GetTrackedWarningSound) _targetattachcomponent)))

(fn ConstructionScript)
