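; Gameplay cue notify for the TrainHeist sentry turret's firing effects, aiming the beam particle at the turret's hit location.
; On burst, it runs the parent behavior, then casts the cue's instigator to the on-train sentry turret parent class.
; It reads that turret's HitLocation variable and sets it as the BeamEnd_0 vector parameter on the first spawned particle component.
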(fn OnBurst(EitherVFX-preferthisversion)
  (params (MyTarget :type Actor%20Object%20Reference) (Parameters :type Gameplay%20Cue%20Parameters%20Structure%20%28by%20ref%29) (ParticleComponents :type Array%20of%20FXSystem%20Component%20Object%20References) (AudioComponents :type Array%20of%20Audio%20Component%20Object%20References) (BurstCameraShakeInstance :type Camera%20Shake%20Base%20Object%20Reference) (BurstDecalInstance :type Decal%20Actor%20Object%20Reference))
  (bind (_normalizedmagnitude _rawmagnitude _effectcontext _matchedtagname _originaltag _aggregatedsourcetags _aggregatedtargettags _location _normal _instigator) (Ability|GameplayCue|BreakGameplayCueParameters Parameters))
  (|CallParentFunction MyTarget Parameters ParticleComponents AudioComponents BurstCameraShakeInstance BurstDecalInstance)
  (bind _asbga_athena_sentry_turret_on_train_parent (Utilities|Casting|CastToBGA_Athena_Sentry_Turret_OnTrain_Parent _instigator))
  (Effects|Components|ParticleSystem|SetVectorParameter (Utilities|Array|Get(acopy) ParticleComponents) "BeamEnd_0" (Variables|/TrainHeist/Gameplay/BGA_Athena_Sentry_Turret_OnTrain_Parent.BGA_Athena_Sentry_Turret_OnTrain_Parent_C|GetHitLocation _asbga_athena_sentry_turret_on_train_parent))
  (return))
